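# Physics asset for the M_MED_DiamondHeart_Chic face accessory part, defining bodies for dynamic hair bones at the front, sides, back and top.
# Each hair bone has a small angled capsule, and hair_base has a smaller capsule as the root.
# Most strands constrain directly to hair_base, while front_02 and front_left_03 chain to front_01 and front_left_02.
# All joints share identical narrow swing limits and a tighter twist limit, restricting hair movement.
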PhysicsAsset /BRCosmetics/Characters/Player/Male/Medium/Bodies/M_MED_DiamondHeart_Chic/Meshes/Parts/M_MED_DiamondHeart_Chic_FaceAcc_Physics
bodies {
 dyn_hair_back_left_02 [{shapeType=Capsule, length=2, radius=1, shapeName=dyn_hair_back_left_02_capsule, rotation=<rot -70.0001 0.0002 -0.0001>}]
 dyn_hair_front_01 [{shapeType=Capsule, length=2, radius=1, shapeName=dyn_hair_front_01_capsule, rotation=<rot -70 -0.0001 0.0001>}]
 dyn_hair_front_02 [{shapeType=Capsule, length=2, radius=1, shapeName=dyn_hair_front_02_capsule, rotation=<rot -70 0.0003 -0.0001>}]
 dyn_hair_front_left_01 [{shapeType=Capsule, length=2, radius=1, shapeName=dyn_hair_front_left_01_capsule, rotation=<rot -70 0.0003 -0.0001>}]
 dyn_hair_front_left_02 [{shapeType=Capsule, length=2, radius=1, shapeName=dyn_hair_front_left_02_capsule, rotation=<rot -70 0.0003 -0.0001>}]
 dyn_hair_front_left_03 [{shapeType=Capsule, length=2, radius=1, shapeName=dyn_hair_front_left_03_capsule, rotation=<rot -70 0.0003 -0.0001>}]
 dyn_hair_front_right_02 [{shapeType=Capsule, length=2, radius=1, shapeName=dyn_hair_front_right_02_capsule, rotation=<rot -70 -0.0002 0>}]
 dyn_hair_side_left__03 [{shapeType=Capsule, length=2, radius=1, shapeName=dyn_hair_side_left__03_capsule, rotation=<rot -70 0.0003 -0.0002>}]
 dyn_hair_side_right_02 [{shapeType=Capsule, length=2, radius=1, shapeName=dyn_hair_side_right_02_capsule, rotation=<rot -70 -0.0002 0.0002>}]
 dyn_hair_top_01 [{shapeType=Capsule, length=2, radius=1, shapeName=dyn_hair_top_01_capsule, rotation=<rot -70 0.0003 -0.0002>}]
 hair_base [{shapeType=Capsule, length=0.505, radius=0.505, shapeName=hair_base_capsule, rotation=<rot 0 -90 0>}]
}
constraints [
 {bone1Name=dyn_hair_top_01, bone2Name=hair_base, swing1LimitDegrees=15, swing2LimitDegrees=15, twistLimitDegrees=2}
 {bone1Name=dyn_hair_back_left_02, bone2Name=hair_base, swing1LimitDegrees=15, swing2LimitDegrees=15, twistLimitDegrees=2}
 {bone1Name=dyn_hair_front_01, bone2Name=hair_base, swing1LimitDegrees=15, swing2LimitDegrees=15, twistLimitDegrees=2}
 {bone1Name=dyn_hair_front_02, bone2Name=dyn_hair_front_01, swing1LimitDegrees=15, swing2LimitDegrees=15, twistLimitDegrees=2}
 {bone1Name=dyn_hair_front_left_01, bone2Name=hair_base, swing1LimitDegrees=15, swing2LimitDegrees=15, twistLimitDegrees=2}
 {bone1Name=dyn_hair_front_left_02, bone2Name=hair_base, swing1LimitDegrees=15, swing2LimitDegrees=15, twistLimitDegrees=2}
 {bone1Name=dyn_hair_front_left_03, bone2Name=dyn_hair_front_left_02, swing1LimitDegrees=15, swing2LimitDegrees=15, twistLimitDegrees=2}
 {bone1Name=dyn_hair_front_right_02, bone2Name=hair_base, swing1LimitDegrees=15, swing2LimitDegrees=15, twistLimitDegrees=2}
 {bone1Name=dyn_hair_side_left__03, bone2Name=hair_base, swing1LimitDegrees=15, swing2LimitDegrees=15, twistLimitDegrees=2}
 {bone1Name=dyn_hair_side_right_02, bone2Name=hair_base, swing1LimitDegrees=15, swing2LimitDegrees=15, twistLimitDegrees=2}
]
